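; BP_WindowPrimDataControl is a Blueprint in the Labs set that tints a window mesh's emissive color per instance.
; Its construction script converts the WindowEmissiveColor variable to a vector and writes it as default Custom Primitive Data on the static mesh component, starting at index 17.
; The color reaches the material as primitive data, with no dynamic material instance created in this script.
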(fn ConstructionScript
  (Rendering|Material|SetDefaultCustomPrimitiveDataVector3 (Variables|Debug|GetStaticMeshComponent) 17 (Math|Conversions|ToVector(LinearColor) (Variables|Default|GetWindowEmissiveColor))))
